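# DV_HoverAnimBP is an Animation Blueprint for Dark Voyager in DragonCartNarrative, on the Fortnite average male player skeleton.
# The Output Pose plays the DragonCart_DarkVoyager_Anim sequence directly.
# A bool blend holding the hover idle sequence, driven by negated IsIdle, is not connected to the output.
# The event graph holds only unconnected update and pawn-owner nodes, so nothing sets IsIdle.
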AnimBlueprint /DragonCartNarrative/DV_HoverAnimBP
target_skeleton=&Fortnite_M_Avg_Player_Skeleton
graphs [
 AnimGraph {nodes [
   Blend Poses by bool(
    BlendPose_1=Sequence Player 'DarkVoyager_Hover_Idle'() {sequence=/DragonCartNarrative/Assets/Anims/DarkVoyager/DarkVoyager_Hover_Idle}
    BlendTime_0=1
    BlendTime_1=1
    bActiveValue=NOT Boolean(A=Get IsIdle() -> (IsIdle:bool))
   )
   Output Pose(Result=Sequence Player 'DragonCart_DarkVoyager_Anim'() {sequence=/DragonCartNarrative/Assets/Anims/DarkVoyager/DragonCart_DarkVoyager_Anim})
  ]}
 EventGraph {nodes [
   TryGetPawnOwner()
   Event BlueprintUpdateAnimation()
  ]}
]
variables [IsIdle {type=boolean}]
